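# Editor utility sub-widget of the Fortnite Quest Editor that shows a title and short usage instructions for creating or opening quest assets.
# It derives from QuestEditorSubWidget and stacks a spacer and two centered text blocks in a vertical box.
# The text points to the 'Create a New Quest Asset' button and the content browser right-click menu.
# The layout is static, with no variables and only empty PreConstruct, Construct and Tick events.
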EditorUtilityWidgetBlueprint /Game/NeverCook/EditorUtilities/QuestTools/QuestEditor/EditorSubWidgets/QuestEditorInstructions
graphs [EventGraph {nodes [
   Event PreConstruct()
   Event Construct()
   Event Tick()
  ]}]
variables []
widgets {
 info {parentClass=/Script/FortniteEditor.QuestEditorSubWidget, rootWidgetClass=&VerticalBox, widgetCount=4}
 widgets [
  VerticalBox_0 {widgetClassPath=&VerticalBox, widget=VerticalBox_0}
  Spacer_66 {widgetClassPath=&Spacer, parent=VerticalBox_0, widget=Spacer_66, properties {size=<1 75>}, slot {class=VerticalBoxSlot}}
  Text_1 {
   widgetClassPath=&CommonTextBlock
   bIsVariable=true
   parent=VerticalBox_0
   widget=Text_1
   properties {
    text=Fortnite Quest Editor
    font {
     bForceMonospaced=false
     bMaterialIsStencil=false
     fontMaterial=null
     fontObject=&NotoSans
     letterSpacing=0
     monospacedWidth=1
     size=15
     skewAmount=0
     typefaceFontName=NotoSans-Bold
     outlineSettings {
      bApplyOutlineToDropShadows=false
      bMiteredCorners=false
      bSeparateFillAlpha=false
      outlineMaterial=null
      outlineSize=0
      outlineColor=<rgba 0 0 0 1>
     }
    }
    shadowOffset=<0 0>
   }
   slot {class=VerticalBoxSlot, properties {horizontalAlignment=HAlign_Center, verticalAlignment=VAlign_Center}}
  }
  Text_2 {
   widgetClassPath=&CommonTextBlock
   bIsVariable=true
   parent=VerticalBox_0
   widget=Text_2
   properties {
    text="Click on the 'Create a New Quest Asset' button to create a new asset.\r\nOpen an existing quest asset from the content browser using the right-click menu."
    font {
     bForceMonospaced=false
     bMaterialIsStencil=false
     fontMaterial=null
     fontObject=&NotoSans
     letterSpacing=0
     monospacedWidth=1
     size=13
     skewAmount=0
     typefaceFontName=NotoSans-Regular
     outlineSettings {
      bApplyOutlineToDropShadows=false
      bMiteredCorners=false
      bSeparateFillAlpha=false
      outlineMaterial=null
      outlineSize=0
      outlineColor=<rgba 0 0 0 1>
     }
    }
    shadowOffset=<0 0>
   }
   slot {class=VerticalBoxSlot, properties {horizontalAlignment=HAlign_Center, verticalAlignment=VAlign_Center, padding=<ltrb 5 5 5 5>}}
  }
 ]
}
