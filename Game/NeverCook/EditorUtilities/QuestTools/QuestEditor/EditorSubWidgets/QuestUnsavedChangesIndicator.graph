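# QuestUnsavedChangesIndicator is an editor sub-widget for the Quest Editor tools, a small marker that, by its name, flags unsaved quest changes.
# It derives from the C++ QuestUnsavedChangesIndicator class and supplies only the layout: a small fixed-size SizeBox holding the inherited Image_Indicator image.
# The event graph has only empty PreConstruct, Construct and Tick events, so any indicator behavior comes from the parent class.
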WidgetBlueprint /Game/NeverCook/EditorUtilities/QuestTools/QuestEditor/EditorSubWidgets/QuestUnsavedChangesIndicator
graphs [EventGraph {nodes [
   Event PreConstruct()
   Event Construct()
   Event Tick()
  ]}]
variables []
widgets {
 info {inheritedWidgetCount=1, parentClass=/Script/FortniteEditor.QuestUnsavedChangesIndicator, rootWidgetClass=&SizeBox, widgetCount=2}
 widgets [
  SizeBox_0 {
   widgetClassPath=&SizeBox
   widget=SizeBox_0
   properties {bOverride_HeightOverride=true, bOverride_WidthOverride=true, heightOverride=10, widthOverride=10}
  }
  Image_Indicator {widgetClassPath=&Image, bInherited=true, bIsVariable=true, parent=SizeBox_0, widget=Image_Indicator, slot {class=SizeBoxSlot}}
 ]
}
