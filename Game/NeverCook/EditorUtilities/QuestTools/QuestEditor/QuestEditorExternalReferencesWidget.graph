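# Editor-only widget for the quest editor that shows controls for a quest's parent quest bundle references.
# A dark bordered row holds an Add Parent Bundle button and a collapsed browse button for the parent bundle asset, above a thin black divider image.
# Both buttons are inherited bindings from FortQuestEditorExternalReferencesWidget.
# The event graph holds only empty event stubs, so the Blueprint supplies just layout and styling.
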WidgetBlueprint /Game/NeverCook/EditorUtilities/QuestTools/QuestEditor/QuestEditorExternalReferencesWidget
graphs [EventGraph {nodes [
   Event PreConstruct()
   Event Construct()
   Event Tick()
  ]}]
variables []
widgets {
 info {
  inheritedWidgetCount=2
  parentClass=/Script/FortniteEditor.FortQuestEditorExternalReferencesWidget
  rootWidgetClass=&Border
  widgetCount=6
 }
 widgets [
  Border_0 {widgetClassPath=&Border, widget=Border_0, properties {brushColor=<rgba 0.0176 0.0176 0.0176 1>, padding=<ltrb 0 0 0 0>}}
  VerticalBox_0 {widgetClassPath=&VerticalBox, parent=Border_0, widget=VerticalBox_0, slot {class=BorderSlot, properties {padding=<ltrb 0 0 0 0>}}}
  HorizontalBox_78 {widgetClassPath=&HorizontalBox, parent=VerticalBox_0, widget=HorizontalBox_78, slot {class=VerticalBoxSlot}}
  Button_AddQuestBundle {
   widgetClassPath=/Game/NeverCook/EditorUtilities/QuestTools/QuestEditor/QuestEditorButtons/QuestEditorButton
   bInherited=true
   bIsVariable=true
   parent=HorizontalBox_78
   widget=Button_AddQuestBundle
   properties {
    bShowIcon=true
    style=/Game/NeverCook/EditorUtilities/QuestTools/QuestEditor/QuestEditorButtons/Styles/QuestEditorButtonStyle_Primary.QuestEditorButtonStyle_Primary_C
    text=Add Parent Bundle
    toolTipText=Add this quest to a parent quest bundle
    iconBrush {
     bIsDynamicallyLoaded=false
     drawAs=Image
     imageType=NoImage
     mirroring=NoMirror
     resourceName=None
     resourceObject=/Game/UI/Foundation/Textures/Buttons/T-SlotIcon-Add-32
     tiling=NoTile
     imageSize=<18 18>
     margin=<ltrb 0 0 0 0>
     outlineSettings {
      bUseBrushTransparency=false
      roundingType=HalfHeightRadius
      width=0
      color {colorUseRule=UseColor_Specified, specifiedColor=<rgba 0 0 0 0>}
      cornerRadii=<0 0 0 0>
     }
     tintColor {colorUseRule=UseColor_Specified, specifiedColor=<rgba 1 1 1 1>}
     uVRegion {bIsValid=false, max=<0 0>, min=<0 0>}
    }
   }
   slot {class=HorizontalBoxSlot, properties {horizontalAlignment=HAlign_Left, verticalAlignment=VAlign_Center, padding=<ltrb 5 5 5 5>}}
  }
  Button_EditQuestBundle {
   widgetClassPath=&QuestEditorButton
   bInherited=true
   bIsVariable=true
   parent=HorizontalBox_78
   widget=Button_EditQuestBundle
   properties {
    bShowIcon=true
    slateIconName=Icons.BrowseContent
    toolTipText=Browse to parent quest bundle asset
    visibility=Collapsed
    iconBrush {
     bIsDynamicallyLoaded=false
     drawAs=NoDrawType
     imageType=NoImage
     mirroring=NoMirror
     resourceName=None
     resourceObject=null
     tiling=Vertical
     imageSize=<32 32>
     margin=<ltrb 0 0 0 0>
     outlineSettings {
      bUseBrushTransparency=false
      roundingType=HalfHeightRadius
      width=0
      color {colorUseRule=UseColor_Specified, specifiedColor=<rgba 0 0 0 0>}
      cornerRadii=<0 0 0 0>
     }
     tintColor {colorUseRule=UseColor_Specified, specifiedColor=<rgba 1 1 1 1>}
     uVRegion {bIsValid=false, max=<0 0>, min=<0 0>}
    }
   }
   slot {class=HorizontalBoxSlot, properties {horizontalAlignment=HAlign_Left, verticalAlignment=VAlign_Center, padding=<ltrb 5 5 5 5>}}
  }
  Spacer_1 {
   widgetClassPath=&Image
   bIsVariable=true
   parent=VerticalBox_0
   widget=Spacer_1
   properties {brush {
     bIsDynamicallyLoaded=false
     drawAs=Image
     imageType=NoImage
     mirroring=NoMirror
     resourceName=None
     resourceObject=null
     tiling=NoTile
     imageSize=<1 1>
     margin=<ltrb 0 0 0 0>
     outlineSettings {
      bUseBrushTransparency=false
      roundingType=HalfHeightRadius
      width=0
      color {colorUseRule=UseColor_Specified, specifiedColor=<rgba 0 0 0 0>}
      cornerRadii=<0 0 0 0>
     }
     tintColor {colorUseRule=UseColor_Specified, specifiedColor=<rgba 0 0 0 1>}
     uVRegion {bIsValid=false, max=<0 0>, min=<0 0>}
    }}
   slot {class=VerticalBoxSlot}
  }
 ]
}
