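# Editor utility widget that lays out a notification entry for the NeverCook quest editor tool.
# It derives from the native QuestEditorNotificationsWidget class, which declares the notification icon and text widgets bound here.
# A dark bordered horizontal box places the icon beside the text.
# The event graph holds only empty PreConstruct, Construct and Tick events, so the Blueprint supplies layout only.
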EditorUtilityWidgetBlueprint /Game/NeverCook/EditorUtilities/QuestTools/QuestEditor/QuestEditorNotificationsWidget
graphs [EventGraph {nodes [
   Event PreConstruct()
   Event Construct()
   Event Tick()
  ]}]
variables []
widgets {
 info {inheritedWidgetCount=2, parentClass=/Script/FortniteEditor.QuestEditorNotificationsWidget, rootWidgetClass=&Border, widgetCount=4}
 widgets [
  Border_0 {widgetClassPath=&Border, widget=Border_0, properties {brushColor=<rgba 0.0075 0.0075 0.0075 1>, padding=<ltrb 5 5 5 5>}}
  HorizontalBox_18 {
   widgetClassPath=&HorizontalBox
   parent=Border_0
   widget=HorizontalBox_18
   slot {class=BorderSlot, properties {padding=<ltrb 5 5 5 5>}}
  }
  Image_Notification {
   widgetClassPath=&Image
   bInherited=true
   bIsVariable=true
   parent=HorizontalBox_18
   widget=Image_Notification
   slot {class=HorizontalBoxSlot, properties {horizontalAlignment=HAlign_Left, verticalAlignment=VAlign_Center, padding=<ltrb 4 4 4 4>}}
  }
  Text_Notification {
   widgetClassPath=&TextBlock
   bInherited=true
   parent=HorizontalBox_18
   widget=Text_Notification
   properties {
    text=Text Block
    font {
     bForceMonospaced=false
     bMaterialIsStencil=false
     fontMaterial=null
     fontObject=&NotoSans
     letterSpacing=0
     monospacedWidth=1
     size=10.5
     skewAmount=0
     typefaceFontName=NotoSans-Regular
     outlineSettings {
      bApplyOutlineToDropShadows=false
      bMiteredCorners=false
      bSeparateFillAlpha=false
      outlineMaterial=null
      outlineSize=0
      outlineColor=<rgba 0 0 0 1>
     }
    }
   }
   slot {class=HorizontalBoxSlot, properties {verticalAlignment=VAlign_Center, padding=<ltrb 4 4 4 4>}}
  }
 ]
}
