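# Editor utility widget that forms the header of the Fortnite quest editor, built on the native FortQuestEditorHeader class.
# A toolbar of buttons saves, browses to, creates, copies, hotfixes and debugs quests, each click calling native functions or asset helpers, casting to FortQuestItemDefinition for clipboard copies.
# It also hosts scope policy, external references, notifications, bulk-edit and hotfix warnings, plus the current quest name with source control and unsaved-change indicators.
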WidgetBlueprint /Game/NeverCook/EditorUtilities/QuestTools/QuestEditor/FortQuestEditorHeader
graphs [EventGraph {nodes [
   "On Clicked (Button_CreateQuestAsset)"!() -> (OutputDelegate, then, Button:Common Button Base Object Reference) {is=title K2Node_ComponentBoundEvent}
   > CreateNewQuestAsset(self:Quest Editor Header Object Reference)
   "On Clicked (Button_SaveQuest)"!() -> (OutputDelegate, then, Button:Common Button Base Object Reference) {is=title K2Node_ComponentBoundEvent}
   > SaveLoadedAsset(AssetToSave=Get AssetObject() -> (AssetObject:Object Reference), bOnlyIfIsDirty=false)
   "On Clicked (Button_BrowseToQuest)"!() -> (OutputDelegate, then, Button:Common Button Base Object Reference) {is=title K2Node_ComponentBoundEvent}
   n1 = Get AssetObject()
   > SyncBrowserToObjects(AssetPaths=Make Array("[0]":String=Get Object Path String(Object=@n1)) -> (Array:Array of Strings))
   "On Clicked (Button_CopyQuestName)"!() -> (OutputDelegate, then, Button:Common Button Base Object Reference) {is=title K2Node_ComponentBoundEvent}
   > n2 = Cast To FortQuestItemDefinition!(execute, Object=Get AssetObject()) -> (then, CastFailed, AsFort Quest Item Definition:Fort Quest Item Definition Object Reference) {is=title K2Node_DynamicCast}
   > CopyQuestNameToClipboard(self:Fort Quest Item Definition Object Reference=@n2)
   "On Clicked (Button_CopyQuestTemplateId)"!() -> (OutputDelegate, then, Button:Common Button Base Object Reference) {is=title K2Node_ComponentBoundEvent}
   > n3 = Cast To FortQuestItemDefinition(Object=Get AssetObject())
   > CopyQuestTemplateIdToClipboard(self:Fort Quest Item Definition Object Reference=@n3)
   "On Clicked (Button_MakeNextInSeries)"!() -> (OutputDelegate, then, Button:Common Button Base Object Reference) {is=title K2Node_ComponentBoundEvent}
   > MakeNextInTheSeries(self:Quest Editor Header Object Reference)
   "On Clicked (Button_OpenQuestDebugger)"!() -> (OutputDelegate, then, Button:Common Button Base Object Reference) {is=title K2Node_ComponentBoundEvent}
   > OpenObjectiveDebugger(self:Fort Quest Editor Header Object Reference)
   "On Clicked (Button_OpenDetailsView)"!() -> (OutputDelegate, then, Button:Common Button Base Object Reference) {is=title K2Node_ComponentBoundEvent}
   > OpenQuestInDetailsView(self:Fort Quest Editor Header Object Reference)
   "On Clicked (Button_HotfixQuest)"!() -> (OutputDelegate, then, Button:Common Button Base Object Reference) {is=title K2Node_ComponentBoundEvent}
   > HotfixQuest(self:Fort Quest Editor Header Object Reference)
   Event PreConstruct()
   Event Construct()
   Event Tick()
  ]}]
variables []
widgets {
 info {inheritedWidgetCount=4, parentClass=/Script/FortniteEditor.FortQuestEditorHeader, rootWidgetClass=&Border, widgetCount=30}
 widgets [
  Border_123 {
   widgetClassPath=&Border
   widget=Border_123
   properties {visibility=SelfHitTestInvisible, brushColor=<rgba 0.0176 0.0176 0.0176 1>, padding=<ltrb 0 0 0 0>}
  }
  VerticalBox_0 {
   widgetClassPath=&VerticalBox
   parent=Border_123
   widget=VerticalBox_0
   slot {class=BorderSlot, properties {padding=<ltrb 0 0 0 0>}}
  }
  ToolBar {widgetClassPath=&HorizontalBox, parent=VerticalBox_0, widget=ToolBar, slot {class=VerticalBoxSlot}}
  Button_SaveQuest {
   widgetClassPath=/Game/NeverCook/EditorUtilities/QuestTools/QuestEditor/QuestEditorButtons/QuestEditorButton
   bIsVariable=true
   parent=ToolBar
   widget=Button_SaveQuest
   properties {
    bShowIcon=true
    slateIconName=Icons.Save
    style=/Game/NeverCook/EditorUtilities/QuestTools/QuestEditor/QuestEditorButtons/Styles/QuestEditorButtonSlateIconStyle.QuestEditorButtonSlateIconStyle_C
    toolTipText=Save current quest
   }
   slot {class=HorizontalBoxSlot, properties {padding=<ltrb 10 5 5 5>}}
  }
  Button_BrowseToQuest {
   widgetClassPath=&QuestEditorButton
   bIsVariable=true
   parent=ToolBar
   widget=Button_BrowseToQuest
   properties {
    bShowIcon=true
    slateIconName=Icons.BrowseContent
    style=&QuestEditorButtonSlateIconStyle_C
    toolTipText=Browses to quest asset
    iconBrush {
     bIsDynamicallyLoaded=false
     drawAs=NoDrawType
     imageType=NoImage
     mirroring=NoMirror
     resourceName=None
     resourceObject=null
     tiling=Vertical
     imageSize=<32 32>
     margin=<ltrb 0 0 0 0>
     outlineSettings {
      bUseBrushTransparency=false
      roundingType=HalfHeightRadius
      width=0
      color {colorUseRule=UseColor_Specified, specifiedColor=<rgba 0 0 0 0>}
      cornerRadii=<0 0 0 0>
     }
     tintColor {colorUseRule=UseColor_Specified, specifiedColor=<rgba 1 1 1 1>}
     uVRegion {bIsValid=false, max=<0 0>, min=<0 0>}
    }
   }
   slot {class=HorizontalBoxSlot, properties {padding=<ltrb 5 5 5 5>}}
  }
  Button_OpenDetailsView {
   widgetClassPath=&QuestEditorButton
   bIsVariable=true
   parent=ToolBar
   widget=Button_OpenDetailsView
   properties {
    bShowIcon=true
    slateIconName=Icons.Edit
    text=Detailed View
    toolTipText=Open in Details View
    iconBrush {
     bIsDynamicallyLoaded=false
     drawAs=NoDrawType
     imageType=NoImage
     mirroring=NoMirror
     resourceName=None
     resourceObject=null
     tiling=Vertical
     imageSize=<32 32>
     margin=<ltrb 0 0 0 0>
     outlineSettings {
      bUseBrushTransparency=false
      roundingType=HalfHeightRadius
      width=0
      color {colorUseRule=UseColor_Specified, specifiedColor=<rgba 0 0 0 0>}
      cornerRadii=<0 0 0 0>
     }
     tintColor {colorUseRule=UseColor_Specified, specifiedColor=<rgba 1 1 1 1>}
     uVRegion {bIsValid=false, max=<0 0>, min=<0 0>}
    }
   }
   slot {class=HorizontalBoxSlot, properties {horizontalAlignment=HAlign_Right, verticalAlignment=VAlign_Top, padding=<ltrb 5 5 5 5>}}
  }
  Button_CreateQuestAsset {
   widgetClassPath=&QuestEditorButton
   bIsVariable=true
   parent=ToolBar
   widget=Button_CreateQuestAsset
   properties {bIsEnabledWithoutValidAsset=true, text=Create New Quest Asset, toolTipText=Create a new quest asset}
   slot {class=HorizontalBoxSlot, properties {padding=<ltrb 5 5 5 5>}}
  }
  Button_MakeNextInSeries {
   widgetClassPath=&QuestEditorButton
   bIsVariable=true
   parent=ToolBar
   widget=Button_MakeNextInSeries
   properties {text=Make Next In Series, toolTipText=Make the next quest in series}
   slot {class=HorizontalBoxSlot, properties {padding=<ltrb 5 5 5 5>}}
  }
  Button_CopyQuestTemplateId {
   widgetClassPath=&QuestEditorButton
   bIsVariable=true
   parent=ToolBar
   widget=Button_CopyQuestTemplateId
   properties {text=Copy Quest TemplateId, toolTipText=Copy quest template ID to clipboard, visibility=Collapsed}
   slot {class=HorizontalBoxSlot, properties {padding=<ltrb 5 5 5 5>}}
  }
  Button_HotfixQuest {
   widgetClassPath=&QuestEditorButton
   bIsVariable=true
   parent=ToolBar
   widget=Button_HotfixQuest
   properties {text=Hotfix Quest, toolTipText=Hotfix Quest, visibility=Visible}
   slot {class=HorizontalBoxSlot, properties {padding=<ltrb 5 5 5 5>}}
  }
  QuestScopePolicyWidget {
   widgetClassPath=/Game/NeverCook/EditorUtilities/QuestTools/QuestEditor/EditorSubWidgets/FortQuestScopePolicyWidget
   bIsVariable=true
   parent=ToolBar
   widget=QuestScopePolicyWidget
   properties {toolTipText="Set the quest scope policy.  Solo, Squad, or Party"}
   slot {class=HorizontalBoxSlot, properties {padding=<ltrb 5 0 5 0>}}
  }
  QuestEditorExternalReferencesWidget {
   widgetClassPath=&/QuestEditorExternalReferencesWidget
   bIsVariable=true
   parent=ToolBar
   widget=QuestEditorExternalReferencesWidget
   slot {class=HorizontalBoxSlot}
  }
  CheckBox_BulkEdit {
   widgetClassPath=/Script/UMG.CheckBox
   bInherited=true
   bIsVariable=true
   parent=ToolBar
   widget=CheckBox_BulkEdit
   properties {horizontalAlignment=HAlign_Center}
   slot {
    class=HorizontalBoxSlot
    properties {horizontalAlignment=HAlign_Right, verticalAlignment=VAlign_Center, padding=<ltrb 5 5 10 5>, size {sizeRule=Fill, value=1}}
   }
  }
  Text_QuestAssetName_1 {
   widgetClassPath=&CommonTextBlock
   parent=CheckBox_BulkEdit
   widget=Text_QuestAssetName_1
   properties {
    text=Bulk Edit
    toolTipText=Name of the current quest
    colorAndOpacity {colorUseRule=UseColor_Specified, specifiedColor=<rgba 0.5271 0.5271 0.5271 1>}
    font {
     bForceMonospaced=false
     bMaterialIsStencil=false
     fontMaterial=null
     fontObject=&NotoSans
     letterSpacing=0
     monospacedWidth=1
     size=9
     skewAmount=0
     typefaceFontName=NotoSans-Regular
     outlineSettings {
      bApplyOutlineToDropShadows=false
      bMiteredCorners=false
      bSeparateFillAlpha=false
      outlineMaterial=null
      outlineSize=0
      outlineColor=<rgba 0 0 0 1>
     }
    }
    shadowOffset=<0 0>
   }
   slot {class=PanelSlot}
  }
  Button_OpenQuestDebugger {
   widgetClassPath=&QuestEditorButton
   bIsVariable=true
   parent=ToolBar
   widget=Button_OpenQuestDebugger
   properties {bIsEnabledWithoutValidAsset=true, text=Verb Debugger, toolTipText=Open Verb Debugger}
   slot {class=HorizontalBoxSlot, properties {horizontalAlignment=HAlign_Right, verticalAlignment=VAlign_Top, padding=<ltrb 5 5 10 5>}}
  }
  Widget_Notifications {
   widgetClassPath=&/QuestEditorNotificationsWidget
   bIsVariable=true
   parent=VerticalBox_0
   widget=Widget_Notifications
   slot {class=VerticalBoxSlot, properties {padding=<ltrb 4 4 4 4>}}
  }
  Border_BulkEditWarning {
   widgetClassPath=&Border
   bInherited=true
   parent=VerticalBox_0
   widget=Border_BulkEditWarning
   properties {visibility=Collapsed, brushColor=<rgba 0.0075 0.0075 0.0075 1>, padding=<ltrb 10 5 5 5>}
   slot {class=VerticalBoxSlot, properties {padding=<ltrb 4 4 4 4>}}
  }
  HorizontalBox_18 {
   widgetClassPath=&HorizontalBox
   parent=Border_BulkEditWarning
   widget=HorizontalBox_18
   slot {class=BorderSlot, properties {padding=<ltrb 10 5 5 5>}}
  }
  Image_Notification {
   widgetClassPath=&Image
   bIsVariable=true
   parent=HorizontalBox_18
   widget=Image_Notification
   properties {brush {
     bIsDynamicallyLoaded=false
     drawAs=Image
     imageType=FullColor
     mirroring=NoMirror
     resourceName=None
     resourceObject=/Game/UI/Foundation/Shell/Textures/T-Icon-Warning-64
     tiling=NoTile
     imageSize=<32 32>
     margin=<ltrb 0 0 0 0>
     outlineSettings {
      bUseBrushTransparency=false
      roundingType=HalfHeightRadius
      width=0
      color {colorUseRule=UseColor_Specified, specifiedColor=<rgba 0 0 0 0>}
      cornerRadii=<0 0 0 0>
     }
     tintColor {colorUseRule=UseColor_Specified, specifiedColor=<rgba 1 0.9279 0.0513 1>}
     uVRegion {bIsValid=false, max=<0 0>, min=<0 0>}
    }}
   slot {class=HorizontalBoxSlot, properties {horizontalAlignment=HAlign_Left, verticalAlignment=VAlign_Center}}
  }
  Text_Notification {
   widgetClassPath=&TextBlock
   parent=HorizontalBox_18
   widget=Text_Notification
   properties {
    text="Bulk Edit is active. Changes made to this quest will be applied to all open quests. "
    font {
     bForceMonospaced=false
     bMaterialIsStencil=false
     fontMaterial=null
     fontObject=&NotoSans
     letterSpacing=0
     monospacedWidth=1
     size=10.5
     skewAmount=0
     typefaceFontName=NotoSans-Regular
     outlineSettings {
      bApplyOutlineToDropShadows=false
      bMiteredCorners=false
      bSeparateFillAlpha=false
      outlineMaterial=null
      outlineSize=0
      outlineColor=<rgba 0 0 0 1>
     }
    }
   }
   slot {class=HorizontalBoxSlot, properties {verticalAlignment=VAlign_Center, padding=<ltrb 4 0 0 0>}}
  }
  QuestEditorHotfixWarningWidget {
   widgetClassPath=/Game/NeverCook/EditorUtilities/QuestTools/QuestEditor/EditorSubWidgets/QuestEditorHotfixWarningWidget
   bIsVariable=true
   parent=VerticalBox_0
   widget=QuestEditorHotfixWarningWidget
   slot {class=VerticalBoxSlot}
  }
  HorizontalBox_0 {widgetClassPath=&HorizontalBox, parent=VerticalBox_0, widget=HorizontalBox_0, slot {class=VerticalBoxSlot}}
  Border_0 {
   widgetClassPath=&Border
   parent=HorizontalBox_0
   widget=Border_0
   properties {brushColor=<rgba 0.0075 0.0075 0.0075 1>, padding=<ltrb 0 0 0 0>}
   slot {class=HorizontalBoxSlot, properties {size {sizeRule=Fill, value=1}}}
  }
  VerticalBox_CurrentlyEditing {
   widgetClassPath=&VerticalBox
   bInherited=true
   parent=Border_0
   widget=VerticalBox_CurrentlyEditing
   slot {class=BorderSlot, properties {padding=<ltrb 0 0 0 0>}}
  }
  QuestEditorSourceControlStatus {
   widgetClassPath=/Game/NeverCook/EditorUtilities/QuestTools/QuestEditor/EditorSubWidgets/QuestEditorSourceControlStatus
   bIsVariable=true
   parent=VerticalBox_CurrentlyEditing
   widget=QuestEditorSourceControlStatus
   slot {class=VerticalBoxSlot, properties {padding=<ltrb 5 0 0 0>}}
  }
  Text_CurrentlyEditing {
   widgetClassPath=&CommonTextBlock
   parent=VerticalBox_CurrentlyEditing
   widget=Text_CurrentlyEditing
   properties {
    text=Currently Editing:
    colorAndOpacity {colorUseRule=UseColor_Specified, specifiedColor=<rgba 0.0953 0.0953 0.0953 1>}
    font {
     bForceMonospaced=false
     bMaterialIsStencil=false
     fontMaterial=null
     fontObject=&NotoSans
     letterSpacing=0
     monospacedWidth=1
     size=9
     skewAmount=0
     typefaceFontName=NotoSans-Italic
     outlineSettings {
      bApplyOutlineToDropShadows=false
      bMiteredCorners=false
      bSeparateFillAlpha=false
      outlineMaterial=null
      outlineSize=0
      outlineColor=<rgba 0 0 0 1>
     }
    }
    shadowOffset=<0 0>
   }
   slot {class=VerticalBoxSlot, properties {horizontalAlignment=HAlign_Left, padding=<ltrb 5 5 0 0>}}
  }
  HorizontalBox_2 {
   widgetClassPath=&HorizontalBox
   parent=VerticalBox_CurrentlyEditing
   widget=HorizontalBox_2
   slot {class=VerticalBoxSlot, properties {padding=<ltrb 5 0 0 0>}}
  }
  Text_QuestAssetName {
   widgetClassPath=&CommonTextBlock
   bInherited=true
   parent=HorizontalBox_2
   widget=Text_QuestAssetName
   properties {
    text=Quest Asset Name
    toolTipText=Name of the current quest
    colorAndOpacity {colorUseRule=UseColor_Specified, specifiedColor=<rgba 1 0.8388 0.0409 1>}
    font {
     bForceMonospaced=false
     bMaterialIsStencil=false
     fontMaterial=null
     fontObject=&NotoSans
     letterSpacing=0
     monospacedWidth=1
     size=16
     skewAmount=0
     typefaceFontName=NotoSans-Regular
     outlineSettings {
      bApplyOutlineToDropShadows=false
      bMiteredCorners=false
      bSeparateFillAlpha=false
      outlineMaterial=null
      outlineSize=0
      outlineColor=<rgba 0 0 0 1>
     }
    }
    shadowOffset=<0 0>
   }
   slot {class=HorizontalBoxSlot, properties {horizontalAlignment=HAlign_Center, verticalAlignment=VAlign_Center, padding=<ltrb 5 0 5 5>}}
  }
  QuestUnsavedChangesIndicator {
   widgetClassPath=/Game/NeverCook/EditorUtilities/QuestTools/QuestEditor/EditorSubWidgets/QuestUnsavedChangesIndicator
   bIsVariable=true
   parent=HorizontalBox_2
   widget=QuestUnsavedChangesIndicator
   slot {class=HorizontalBoxSlot, properties {horizontalAlignment=HAlign_Center, verticalAlignment=VAlign_Top, padding=<ltrb 5 0 5 5>}}
  }
  Button_CopyQuestName {
   widgetClassPath=&QuestEditorButton
   bIsVariable=true
   parent=HorizontalBox_2
   widget=Button_CopyQuestName
   properties {bShowIcon=true, slateIconName=GenericCommands.Copy, toolTipText=Copy quest name to clipboard}
   slot {class=HorizontalBoxSlot, properties {horizontalAlignment=HAlign_Center, verticalAlignment=VAlign_Top, padding=<ltrb 5 0 5 5>}}
  }
 ]
}
