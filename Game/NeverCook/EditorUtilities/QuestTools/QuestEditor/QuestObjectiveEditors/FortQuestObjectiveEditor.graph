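# Editor utility widget for editing a single quest objective in the Fortnite quest editor tools, built on the C++ FortQuestObjectiveEditor class.
# An expandable header shows the objective's label, backend name and library preview, with copy, hotfix, duplicate and remove buttons.
# The body holds a multi-column objective property view, an extension tabbed view and a filters editor.
# The graph only wires Copy to put the backend name on the clipboard and Create Hotfix to call CreateObjectiveHotfix.
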EditorUtilityWidgetBlueprint /Game/NeverCook/EditorUtilities/QuestTools/QuestEditor/QuestObjectiveEditors/FortQuestObjectiveEditor
graphs [EventGraph {nodes [
   "On Clicked (Button_CopyQuestName)"!() -> (OutputDelegate, then, Button:Common Button Base Object Reference) {is=title K2Node_ComponentBoundEvent}
   > CopyStringToClipboard!(execute, InString:String="To String (Name)"(InName=Get ObjectiveBackendName() -> (ObjectiveBackendName:Name=None))) -> (then)
   "On Clicked (Button_CreateHotfix)"!() -> (OutputDelegate, then, Button:Common Button Base Object Reference) {is=title K2Node_ComponentBoundEvent}
   > CreateObjectiveHotfix(self:Fort Quest Objective Editor Object Reference)
   Event PreConstruct()
   Event Construct()
   Event Tick()
  ]}]
variables []
widgets {
 info {
  inheritedWidgetCount=10
  namedSlotCount=2
  parentClass=/Script/FortniteEditor.FortQuestObjectiveEditor
  rootWidgetClass=&Border
  widgetCount=20
 }
 widgets [
  Border_67 {
   widgetClassPath=&Border
   widget=Border_67
   properties {
    background {
     bIsDynamicallyLoaded=false
     drawAs=Image
     imageType=NoImage
     mirroring=NoMirror
     resourceName=None
     resourceObject=null
     tiling=NoTile
     imageSize=<32 32>
     margin=<ltrb 0 0 0 0>
     outlineSettings {
      bUseBrushTransparency=false
      roundingType=HalfHeightRadius
      width=0
      color {colorUseRule=UseColor_Specified, specifiedColor=<rgba 0 0 0 0>}
      cornerRadii=<0 0 0 0>
     }
     tintColor {colorUseRule=UseColor_Specified, specifiedColor=<rgba 0.0103 0.0103 0.0103 1>}
     uVRegion {bIsValid=false, max=<0 0>, min=<0 0>}
    }
    padding=<ltrb 4 0 4 0>
   }
  }
  VerticalBox_1 {
   widgetClassPath=&VerticalBox
   parent=Border_67
   widget=VerticalBox_1
   slot {class=BorderSlot, properties {padding=<ltrb 4 0 4 0>}}
  }
  ExpandableArea_ObjectiveArea {
   widgetClassPath=/Script/UMG.ExpandableArea
   bInherited=true
   bIsVariable=true
   parent=VerticalBox_1
   widget=ExpandableArea_ObjectiveArea
   properties {
    bIsExpanded=true
    areaPadding=<ltrb 1 1 0 0>
    borderBrush {
     bIsDynamicallyLoaded=false
     drawAs=Border
     imageType=NoImage
     mirroring=NoMirror
     resourceName=None
     resourceObject=null
     tiling=NoTile
     imageSize=<0 0>
     margin=<ltrb 0 0 0 0>
     outlineSettings {
      bUseBrushTransparency=false
      roundingType=FixedRadius
      width=0
      color {colorUseRule=UseColor_Specified, specifiedColor=<rgba 0 0 0 0>}
      cornerRadii=<4 4 4 4>
     }
     tintColor {colorUseRule=UseColor_Specified, specifiedColor=<rgba 0.0176 0.0176 0.0176 1>}
     uVRegion {bIsValid=false, max=<0 0>, min=<0 0>}
    }
    headerPadding=<ltrb 4 0 4 0>
    style {
     rolloutAnimationSeconds=0.1
     collapsedImage {
      bIsDynamicallyLoaded=false
      drawAs=Image
      imageType=Vector
      mirroring=NoMirror
      resourceName=../../../Engine/Content/Slate/Starship/Common/caret-right.svg
      resourceObject=null
      tiling=NoTile
      imageSize=<16 16>
      margin=<ltrb 0 0 0 0>
      outlineSettings {
       bUseBrushTransparency=false
       roundingType=HalfHeightRadius
       width=0
       color {colorUseRule=UseColor_Specified, specifiedColor=<rgba 0 0 0 0>}
       cornerRadii=<0 0 0 0>
      }
      tintColor {colorUseRule=UseColor_Specified, specifiedColor=<rgba 0.5271 0.5271 0.5271 1>}
      uVRegion {bIsValid=false, max=<0 0>, min=<0 0>}
     }
     expandedImage {
      bIsDynamicallyLoaded=false
      drawAs=Image
      imageType=Vector
      mirroring=NoMirror
      resourceName=../../../Engine/Content/Slate/Starship/Common/caret-down.svg
      resourceObject=null
      tiling=NoTile
      imageSize=<16 16>
      margin=<ltrb 0 0 0 0>
      outlineSettings {
       bUseBrushTransparency=false
       roundingType=HalfHeightRadius
       width=0
       color {colorUseRule=UseColor_Specified, specifiedColor=<rgba 0 0 0 0>}
       cornerRadii=<0 0 0 0>
      }
      tintColor {colorUseRule=UseColor_Specified, specifiedColor=<rgba 0.5271 0.5271 0.5271 1>}
      uVRegion {bIsValid=false, max=<0 0>, min=<0 0>}
     }
    }
   }
   slot {class=VerticalBoxSlot}
  }
  HorizontalBox_3 {widgetClassPath=&HorizontalBox, namedSlotHost=ExpandableArea_ObjectiveArea, widget=HorizontalBox_3}
  Text_ObjectiveHeader {
   widgetClassPath=&CommonTextBlock
   bInherited=true
   bIsVariable=true
   parent=HorizontalBox_3
   widget=Text_ObjectiveHeader
   properties {
    text="Objective {0}:"
    font {
     bForceMonospaced=false
     bMaterialIsStencil=false
     fontMaterial=null
     fontObject=&NotoSans
     letterSpacing=0
     monospacedWidth=1
     size=13
     skewAmount=0
     typefaceFontName=NotoSans-Bold
     outlineSettings {
      bApplyOutlineToDropShadows=false
      bMiteredCorners=false
      bSeparateFillAlpha=false
      outlineMaterial=null
      outlineSize=0
      outlineColor=<rgba 0 0 0 1>
     }
    }
    shadowOffset=<0 0>
   }
   slot {class=HorizontalBoxSlot, properties {horizontalAlignment=HAlign_Center, verticalAlignment=VAlign_Center}}
  }
  Text_ObjectiveBackendName {
   widgetClassPath=&CommonTextBlock
   bInherited=true
   bIsVariable=true
   parent=HorizontalBox_3
   widget=Text_ObjectiveBackendName
   properties {
    text=Backend Name
    colorAndOpacity {colorUseRule=UseColor_Specified, specifiedColor=<rgba 0.0685 0.0685 0.0685 1>}
    font {
     bForceMonospaced=false
     bMaterialIsStencil=false
     fontMaterial=null
     fontObject=&NotoSans
     letterSpacing=0
     monospacedWidth=1
     size=10
     skewAmount=0
     typefaceFontName=NotoSans-Bold
     outlineSettings {
      bApplyOutlineToDropShadows=false
      bMiteredCorners=false
      bSeparateFillAlpha=false
      outlineMaterial=null
      outlineSize=0
      outlineColor=<rgba 0 0 0 1>
     }
    }
    shadowOffset=<0 0>
   }
   slot {class=HorizontalBoxSlot, properties {verticalAlignment=VAlign_Center, padding=<ltrb 7 0 0 0>}}
  }
  DetailView_ObjectiveLabel {
   widgetClassPath=/Script/PropertiesDetailView
   bInherited=true
   bIsVariable=true
   parent=HorizontalBox_3
   widget=DetailView_ObjectiveLabel
   properties {
    bShowPropertiesThatMatchGeneratedPropertyPaths=true
    visibility=SelfHitTestInvisible
    propertiesToExclude [ObjectiveStatHandle]
    propertiesToSkipPropertyPathGeneration [Objectives]
    propertyColumns [{
      fillCoefficient=1
      propertiesDisplayData [{
        bForceShowParentProperty=false
        bForceShowProperty=false
        bShowChildren=true
        bShowContainerProperty=true
        bShowName=true
        bShowPropertyButtons=true
        bShowValue=true
        containerObjectName=None
        propertyName=ObjectiveLabel
        propertyNameOverride=""
       }]
     }]
   }
   slot {class=HorizontalBoxSlot, properties {horizontalAlignment=HAlign_Center, verticalAlignment=VAlign_Center, padding=<ltrb 5 5 5 5>}}
  }
  Button_CopyQuestName {
   widgetClassPath=/Game/NeverCook/EditorUtilities/QuestTools/QuestEditor/QuestEditorButtons/QuestEditorButton
   bIsVariable=true
   parent=HorizontalBox_3
   widget=Button_CopyQuestName
   properties {bShowIcon=true, slateIconName=GenericCommands.Copy, toolTipText=Copy objective backend name to clipboard}
   slot {class=HorizontalBoxSlot, properties {horizontalAlignment=HAlign_Center, verticalAlignment=VAlign_Center, padding=<ltrb 5 0 0 0>}}
  }
  Widget_QuestLibraryPreview {
   widgetClassPath=/Script/FortniteEditor.QuestLibraryPreview
   bInherited=true
   bIsVariable=true
   parent=HorizontalBox_3
   widget=Widget_QuestLibraryPreview
   slot {class=HorizontalBoxSlot, properties {verticalAlignment=VAlign_Center, padding=<ltrb 10 0 0 0>, size {sizeRule=Fill, value=1}}}
  }
  Spacer_64 {
   widgetClassPath=&Spacer
   parent=HorizontalBox_3
   widget=Spacer_64
   slot {class=HorizontalBoxSlot, properties {size {sizeRule=Fill, value=1}}}
  }
  Button_CreateHotfix {
   widgetClassPath=&QuestEditorButton
   bIsVariable=true
   parent=HorizontalBox_3
   widget=Button_CreateHotfix
   properties {slateIconName=GenericCommands.Duplicate, text=Create Hotfix, toolTipText=Generates syntax for objective hotfix}
   slot {class=HorizontalBoxSlot, properties {horizontalAlignment=HAlign_Right, verticalAlignment=VAlign_Top, padding=<ltrb 5 5 5 5>}}
  }
  Button_DuplicateObjective {
   widgetClassPath=&QuestEditorButton
   bInherited=true
   bIsVariable=true
   parent=HorizontalBox_3
   widget=Button_DuplicateObjective
   properties {bShowIcon=true, slateIconName=GenericCommands.Duplicate, text=Duplicate Objective, toolTipText=Make a copy of the objective in the quest}
   slot {class=HorizontalBoxSlot, properties {horizontalAlignment=HAlign_Right, verticalAlignment=VAlign_Top, padding=<ltrb 5 5 5 5>}}
  }
  Button_RemoveObjective {
   widgetClassPath=&QuestEditorButton
   bInherited=true
   bIsVariable=true
   parent=HorizontalBox_3
   widget=Button_RemoveObjective
   properties {
    bShowIcon=true
    text=Remove Objective
    toolTipText="Remove objective from the quest "
    iconBrush {
     bIsDynamicallyLoaded=false
     drawAs=Image
     imageType=NoImage
     mirroring=NoMirror
     resourceName=None
     resourceObject=/Game/UI/Foundation/Textures/Buttons/T-SlotIcon-Remove-32
     tiling=NoTile
     imageSize=<16 16>
     margin=<ltrb 0 0 0 0>
     outlineSettings {
      bUseBrushTransparency=false
      roundingType=HalfHeightRadius
      width=0
      color {colorUseRule=UseColor_Specified, specifiedColor=<rgba 0 0 0 0>}
      cornerRadii=<0 0 0 0>
     }
     tintColor {colorUseRule=UseColor_Specified, specifiedColor=<rgba 1 0.0017 0.0017 1>}
     uVRegion {bIsValid=false, max=<0 0>, min=<0 0>}
    }
   }
   slot {class=HorizontalBoxSlot, properties {horizontalAlignment=HAlign_Right, verticalAlignment=VAlign_Top, padding=<ltrb 5 5 0 5>}}
  }
  Border_Root {
   widgetClassPath=&Border
   bIsVariable=true
   namedSlotHost=ExpandableArea_ObjectiveArea
   widget=Border_Root
   properties {brushColor=<rgba 0.0685 0.0685 0.0685 1>, padding=<ltrb 4 4 4 2>}
  }
  VerticalBox_0 {
   widgetClassPath=&VerticalBox
   parent=Border_Root
   widget=VerticalBox_0
   slot {class=BorderSlot, properties {padding=<ltrb 4 4 4 2>}}
  }
  Border_Properties {
   widgetClassPath=&Border
   parent=VerticalBox_0
   widget=Border_Properties
   properties {brushColor=<rgba 0.0284 0.0284 0.0284 1>, padding=<ltrb 2 2 2 2>}
   slot {class=VerticalBoxSlot}
  }
  ScrollBox_0 {
   widgetClassPath=&ScrollBox
   parent=Border_Properties
   widget=ScrollBox_0
   properties {
    orientation=Orient_Horizontal
    widgetBarStyle {
     thickness=8
     draggedThumbImage {
      bIsDynamicallyLoaded=false
      drawAs=RoundedBox
      imageType=NoImage
      mirroring=NoMirror
      resourceName=None
      resourceObject=null
      tiling=NoTile
      imageSize=<0 0>
      margin=<ltrb 0 0 0 0>
      outlineSettings {
       bUseBrushTransparency=false
       roundingType=FixedRadius
       width=0
       color {colorUseRule=UseColor_Specified, specifiedColor=<rgba 0 0 0 0>}
       cornerRadii=<4 4 4 4>
      }
      tintColor {colorUseRule=UseColor_Specified, specifiedColor=<rgba 0.2159 0.2159 0.2159 1>}
      uVRegion {bIsValid=false, max=<0 0>, min=<0 0>}
     }
     horizontalBackgroundImage {
      bIsDynamicallyLoaded=false
      drawAs=NoDrawType
      imageType=NoImage
      mirroring=NoMirror
      resourceName=None
      resourceObject=null
      tiling=NoTile
      imageSize=<0 0>
      margin=<ltrb 0 0 0 0>
      outlineSettings {
       bUseBrushTransparency=false
       roundingType=HalfHeightRadius
       width=0
       color {colorUseRule=UseColor_Specified, specifiedColor=<rgba 0 0 0 0>}
       cornerRadii=<0 0 0 0>
      }
      tintColor {colorUseRule=UseColor_Specified, specifiedColor=<rgba 1 1 1 1>}
      uVRegion {bIsValid=false, max=<0 0>, min=<0 0>}
     }
     horizontalBottomSlotImage {
      bIsDynamicallyLoaded=false
      drawAs=NoDrawType
      imageType=NoImage
      mirroring=NoMirror
      resourceName=None
      resourceObject=null
      tiling=NoTile
      imageSize=<0 0>
      margin=<ltrb 0 0 0 0>
      outlineSettings {
       bUseBrushTransparency=false
       roundingType=HalfHeightRadius
       width=0
       color {colorUseRule=UseColor_Specified, specifiedColor=<rgba 0 0 0 0>}
       cornerRadii=<0 0 0 0>
      }
      tintColor {colorUseRule=UseColor_Specified, specifiedColor=<rgba 1 1 1 1>}
      uVRegion {bIsValid=false, max=<0 0>, min=<0 0>}
     }
     horizontalTopSlotImage {
      bIsDynamicallyLoaded=false
      drawAs=NoDrawType
      imageType=NoImage
      mirroring=NoMirror
      resourceName=None
      resourceObject=null
      tiling=NoTile
      imageSize=<0 0>
      margin=<ltrb 0 0 0 0>
      outlineSettings {
       bUseBrushTransparency=false
       roundingType=HalfHeightRadius
       width=0
       color {colorUseRule=UseColor_Specified, specifiedColor=<rgba 0 0 0 0>}
       cornerRadii=<0 0 0 0>
      }
      tintColor {colorUseRule=UseColor_Specified, specifiedColor=<rgba 1 1 1 1>}
      uVRegion {bIsValid=false, max=<0 0>, min=<0 0>}
     }
     hoveredThumbImage {
      bIsDynamicallyLoaded=false
      drawAs=RoundedBox
      imageType=NoImage
      mirroring=NoMirror
      resourceName=None
      resourceObject=null
      tiling=NoTile
      imageSize=<0 0>
      margin=<ltrb 0 0 0 0>
      outlineSettings {
       bUseBrushTransparency=false
       roundingType=FixedRadius
       width=0
       color {colorUseRule=UseColor_Specified, specifiedColor=<rgba 0 0 0 0>}
       cornerRadii=<4 4 4 4>
      }
      tintColor {colorUseRule=UseColor_Specified, specifiedColor=<rgba 0.2159 0.2159 0.2159 1>}
      uVRegion {bIsValid=false, max=<0 0>, min=<0 0>}
     }
     normalThumbImage {
      bIsDynamicallyLoaded=false
      drawAs=RoundedBox
      imageType=NoImage
      mirroring=NoMirror
      resourceName=None
      resourceObject=null
      tiling=NoTile
      imageSize=<0 0>
      margin=<ltrb 0 0 0 0>
      outlineSettings {
       bUseBrushTransparency=false
       roundingType=FixedRadius
       width=0
       color {colorUseRule=UseColor_Specified, specifiedColor=<rgba 0 0 0 0>}
       cornerRadii=<4 4 4 4>
      }
      tintColor {colorUseRule=UseColor_Specified, specifiedColor=<rgba 0.0953 0.0953 0.0953 1>}
      uVRegion {bIsValid=false, max=<0 0>, min=<0 0>}
     }
     verticalBackgroundImage {
      bIsDynamicallyLoaded=false
      drawAs=NoDrawType
      imageType=NoImage
      mirroring=NoMirror
      resourceName=None
      resourceObject=null
      tiling=NoTile
      imageSize=<0 0>
      margin=<ltrb 0 0 0 0>
      outlineSettings {
       bUseBrushTransparency=false
       roundingType=HalfHeightRadius
       width=0
       color {colorUseRule=UseColor_Specified, specifiedColor=<rgba 0 0 0 0>}
       cornerRadii=<0 0 0 0>
      }
      tintColor {colorUseRule=UseColor_Specified, specifiedColor=<rgba 1 1 1 1>}
      uVRegion {bIsValid=false, max=<0 0>, min=<0 0>}
     }
     verticalBottomSlotImage {
      bIsDynamicallyLoaded=false
      drawAs=NoDrawType
      imageType=NoImage
      mirroring=NoMirror
      resourceName=None
      resourceObject=null
      tiling=NoTile
      imageSize=<0 0>
      margin=<ltrb 0 0 0 0>
      outlineSettings {
       bUseBrushTransparency=false
       roundingType=HalfHeightRadius
       width=0
       color {colorUseRule=UseColor_Specified, specifiedColor=<rgba 0 0 0 0>}
       cornerRadii=<0 0 0 0>
      }
      tintColor {colorUseRule=UseColor_Specified, specifiedColor=<rgba 1 1 1 1>}
      uVRegion {bIsValid=false, max=<0 0>, min=<0 0>}
     }
     verticalTopSlotImage {
      bIsDynamicallyLoaded=false
      drawAs=NoDrawType
      imageType=NoImage
      mirroring=NoMirror
      resourceName=None
      resourceObject=null
      tiling=NoTile
      imageSize=<0 0>
      margin=<ltrb 0 0 0 0>
      outlineSettings {
       bUseBrushTransparency=false
       roundingType=HalfHeightRadius
       width=0
       color {colorUseRule=UseColor_Specified, specifiedColor=<rgba 0 0 0 0>}
       cornerRadii=<0 0 0 0>
      }
      tintColor {colorUseRule=UseColor_Specified, specifiedColor=<rgba 1 1 1 1>}
      uVRegion {bIsValid=false, max=<0 0>, min=<0 0>}
     }
    }
    widgetStyle {
     barThickness=8
     bottomShadowBrush {
      bIsDynamicallyLoaded=false
      drawAs=Box
      imageType=FullColor
      mirroring=NoMirror
      resourceName=../../../Engine/Content/Slate/Common/ScrollBoxShadowBottom.png
      resourceObject=null
      tiling=NoTile
      imageSize=<16 8>
      margin=<ltrb 0.5 0 0.5 1>
      outlineSettings {
       bUseBrushTransparency=false
       roundingType=HalfHeightRadius
       width=0
       color {colorUseRule=UseColor_Specified, specifiedColor=<rgba 0 0 0 0>}
       cornerRadii=<0 0 0 0>
      }
      tintColor {colorUseRule=UseColor_Specified, specifiedColor=<rgba 1 1 1 1>}
      uVRegion {bIsValid=false, max=<0 0>, min=<0 0>}
     }
     horizontalScrolledContentPadding=<ltrb 0 0 1 0>
     leftShadowBrush {
      bIsDynamicallyLoaded=false
      drawAs=Box
      imageType=FullColor
      mirroring=NoMirror
      resourceName=../../../Engine/Content/Slate/Common/ScrollBoxShadowLeft.png
      resourceObject=null
      tiling=NoTile
      imageSize=<8 16>
      margin=<ltrb 1 0.5 0 0.5>
      outlineSettings {
       bUseBrushTransparency=false
       roundingType=HalfHeightRadius
       width=0
       color {colorUseRule=UseColor_Specified, specifiedColor=<rgba 0 0 0 0>}
       cornerRadii=<0 0 0 0>
      }
      tintColor {colorUseRule=UseColor_Specified, specifiedColor=<rgba 1 1 1 1>}
      uVRegion {bIsValid=false, max=<0 0>, min=<0 0>}
     }
     rightShadowBrush {
      bIsDynamicallyLoaded=false
      drawAs=Box
      imageType=FullColor
      mirroring=NoMirror
      resourceName=../../../Engine/Content/Slate/Common/ScrollBoxShadowRight.png
      resourceObject=null
      tiling=NoTile
      imageSize=<8 16>
      margin=<ltrb 0 0.5 1 0.5>
      outlineSettings {
       bUseBrushTransparency=false
       roundingType=HalfHeightRadius
       width=0
       color {colorUseRule=UseColor_Specified, specifiedColor=<rgba 0 0 0 0>}
       cornerRadii=<0 0 0 0>
      }
      tintColor {colorUseRule=UseColor_Specified, specifiedColor=<rgba 1 1 1 1>}
      uVRegion {bIsValid=false, max=<0 0>, min=<0 0>}
     }
     topShadowBrush {
      bIsDynamicallyLoaded=false
      drawAs=Box
      imageType=FullColor
      mirroring=NoMirror
      resourceName=../../../Engine/Content/Slate/Common/ScrollBoxShadowTop.png
      resourceObject=null
      tiling=NoTile
      imageSize=<16 8>
      margin=<ltrb 0.5 1 0.5 0>
      outlineSettings {
       bUseBrushTransparency=false
       roundingType=HalfHeightRadius
       width=0
       color {colorUseRule=UseColor_Specified, specifiedColor=<rgba 0 0 0 0>}
       cornerRadii=<0 0 0 0>
      }
      tintColor {colorUseRule=UseColor_Specified, specifiedColor=<rgba 1 1 1 1>}
      uVRegion {bIsValid=false, max=<0 0>, min=<0 0>}
     }
     verticalScrolledContentPadding=<ltrb 0 0 0 1>
    }
   }
   slot {class=BorderSlot, properties {padding=<ltrb 2 2 2 2>}}
  }
  DetailView_Objective {
   widgetClassPath=&PropertiesDetailView
   bInherited=true
   bIsVariable=true
   parent=ScrollBox_0
   widget=DetailView_Objective
   properties {
    visibility=SelfHitTestInvisible
    propertiesToExclude [ObjectiveStatHandle]
    propertiesToSkipPropertyPathGeneration [Objectives]
    propertyColumns [
     {
      fillCoefficient=1
      propertiesDisplayData [
       {
        bForceShowParentProperty=false
        bForceShowProperty=false
        bShowChildren=true
        bShowContainerProperty=true
        bShowName=true
        bShowPropertyButtons=true
        bShowValue=true
        containerObjectName=None
        propertyName=Count
        propertyNameOverride=""
       }
       {
        bForceShowParentProperty=false
        bForceShowProperty=false
        bShowChildren=true
        bShowContainerProperty=true
        bShowName=true
        bShowPropertyButtons=true
        bShowValue=true
        containerObjectName=None
        propertyName=None
        propertyNameOverride=""
       }
      ]
     }
     {
      fillCoefficient=1
      propertiesDisplayData [
       {
        bForceShowParentProperty=false
        bForceShowProperty=false
        bShowChildren=true
        bShowContainerProperty=true
        bShowName=true
        bShowPropertyButtons=true
        bShowValue=true
        containerObjectName=None
        propertyName=bDisplayDynamicAnnouncementUpdate
        propertyNameOverride=""
       }
       {
        bForceShowParentProperty=false
        bForceShowProperty=false
        bShowChildren=true
        bShowContainerProperty=true
        bShowName=true
        bShowPropertyButtons=true
        bShowValue=true
        containerObjectName=None
        propertyName=DynamicStatusUpdateType
        propertyNameOverride=""
       }
       {
        bForceShowParentProperty=false
        bForceShowProperty=false
        bShowChildren=true
        bShowContainerProperty=true
        bShowName=true
        bShowPropertyButtons=true
        bShowValue=true
        containerObjectName=None
        propertyName=DynamicStatusUpdatePercentInterval
        propertyNameOverride=""
       }
      ]
     }
     {
      fillCoefficient=1
      propertiesDisplayData [
       {
        bForceShowParentProperty=false
        bForceShowProperty=false
        bShowChildren=true
        bShowContainerProperty=true
        bShowName=true
        bShowPropertyButtons=true
        bShowValue=true
        containerObjectName=None
        propertyName=Description
        propertyNameOverride=""
       }
       {
        bForceShowParentProperty=false
        bForceShowProperty=false
        bShowChildren=true
        bShowContainerProperty=true
        bShowName=true
        bShowPropertyButtons=true
        bShowValue=true
        containerObjectName=None
        propertyName=HudShortDescription
        propertyNameOverride=""
       }
      ]
     }
     {
      fillCoefficient=1
      propertiesDisplayData [
       {
        bForceShowParentProperty=false
        bForceShowProperty=false
        bShowChildren=false
        bShowContainerProperty=true
        bShowName=true
        bShowPropertyButtons=true
        bShowValue=true
        containerObjectName=None
        propertyName=DisplayStage
        propertyNameOverride=""
       }
       {
        bForceShowParentProperty=false
        bForceShowProperty=false
        bShowChildren=false
        bShowContainerProperty=true
        bShowName=true
        bShowPropertyButtons=true
        bShowValue=true
        containerObjectName=None
        propertyName=DisplayMaxStage
        propertyNameOverride=""
       }
      ]
     }
    ]
   }
   slot {class=ScrollBoxSlot, properties {size {sizeRule=Fill, value=1}}}
  }
  TabbedView_ObjectiveExtension {
   widgetClassPath=/Game/NeverCook/EditorUtilities/QuestTools/QuestEditor/QuestItemDefComponentUI/FortQuestItemDefComponentObjectiveExtensionTabbedView
   bInherited=true
   bIsVariable=true
   parent=VerticalBox_0
   widget=TabbedView_ObjectiveExtension
   slot {class=VerticalBoxSlot, properties {padding=<ltrb 0 4 0 4>, size {sizeRule=Fill, value=1}}}
  }
  Editor_ObjectiveFilters {
   widgetClassPath=/Game/NeverCook/EditorUtilities/QuestTools/QuestEditor/QuestObjectiveEditors/ObjectiveFilterEditors/FortObjectiveFiltersEditor
   bInherited=true
   bIsVariable=true
   parent=VerticalBox_0
   widget=Editor_ObjectiveFilters
   slot {class=VerticalBoxSlot, properties {padding=<ltrb 16 0 16 0>}}
  }
 ]
}
