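# Editor utility toolbar widget for quest objectives in the quest editor, derived from the C++ FortQuestObjectiveToolbar class.
# A horizontal row holds the inherited New Objective button and objective count label, a divider, a common tag query menu and a documentation button, above a thin separator line.
# The only wired Blueprint logic opens a Confluence guide on adding tag queries when the help button is clicked.
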WidgetBlueprint /Game/NeverCook/EditorUtilities/QuestTools/QuestEditor/QuestObjectiveEditors/FortQuestObjectiveToolbar
graphs [EventGraph {nodes [
   "On Clicked (Button_CommonTagQueriesHelp)"!() -> (OutputDelegate, then, Button:Common Button Base Object Reference) {is=title K2Node_ComponentBoundEvent}
   > LaunchURL!(execute, URL:String=https://confluence.it.epicgames.com/display/FORT/How+to+add+Common+Tags+to+QS2+Quest+Editor) -> (then)
   Event PreConstruct()
   Event Construct()
   Event Tick()
  ]}]
variables []
widgets {
 info {inheritedWidgetCount=2, parentClass=/Script/FortniteEditor.FortQuestObjectiveToolbar, rootWidgetClass=&Border, widgetCount=9}
 widgets [
  Border_0 {widgetClassPath=&Border, widget=Border_0, properties {brushColor=<rgba 0.0176 0.0176 0.0176 1>, padding=<ltrb 0 0 0 0>}}
  VerticalBox_0 {widgetClassPath=&VerticalBox, parent=Border_0, widget=VerticalBox_0, slot {class=BorderSlot, properties {padding=<ltrb 0 0 0 0>}}}
  HorizontalBox_78 {widgetClassPath=&HorizontalBox, parent=VerticalBox_0, widget=HorizontalBox_78, slot {class=VerticalBoxSlot}}
  Button_AddObjective {
   widgetClassPath=/Game/NeverCook/EditorUtilities/QuestTools/QuestEditor/QuestEditorButtons/QuestEditorButton
   bInherited=true
   bIsVariable=true
   parent=HorizontalBox_78
   widget=Button_AddObjective
   properties {
    bShowIcon=true
    style=/Game/NeverCook/EditorUtilities/QuestTools/QuestEditor/QuestEditorButtons/Styles/QuestEditorButtonStyle_Primary.QuestEditorButtonStyle_Primary_C
    text=New Objective
    toolTipText=Create a new objective in the quest
    iconBrush {
     bIsDynamicallyLoaded=false
     drawAs=Image
     imageType=NoImage
     mirroring=NoMirror
     resourceName=None
     resourceObject=/Game/UI/Foundation/Textures/Buttons/T-SlotIcon-Add-32
     tiling=NoTile
     imageSize=<18 18>
     margin=<ltrb 0 0 0 0>
     outlineSettings {
      bUseBrushTransparency=false
      roundingType=HalfHeightRadius
      width=0
      color {colorUseRule=UseColor_Specified, specifiedColor=<rgba 0 0 0 0>}
      cornerRadii=<0 0 0 0>
     }
     tintColor {colorUseRule=UseColor_Specified, specifiedColor=<rgba 1 1 1 1>}
     uVRegion {bIsValid=false, max=<0 0>, min=<0 0>}
    }
   }
   slot {class=HorizontalBoxSlot, properties {horizontalAlignment=HAlign_Left, verticalAlignment=VAlign_Center, padding=<ltrb 10 5 5 5>}}
  }
  Text_NumObjectives {
   widgetClassPath=&CommonTextBlock
   bInherited=true
   parent=HorizontalBox_78
   widget=Text_NumObjectives
   properties {
    text="Objective Count: "
    toolTipText=Number of objectives in the quest
    font {
     bForceMonospaced=false
     bMaterialIsStencil=false
     fontMaterial=null
     fontObject=&NotoSans
     letterSpacing=0
     monospacedWidth=1
     size=10
     skewAmount=0
     typefaceFontName=NotoSans-Regular
     outlineSettings {
      bApplyOutlineToDropShadows=false
      bMiteredCorners=false
      bSeparateFillAlpha=false
      outlineMaterial=null
      outlineSize=0
      outlineColor=<rgba 0 0 0 1>
     }
    }
    shadowOffset=<0 0>
   }
   slot {class=HorizontalBoxSlot, properties {horizontalAlignment=HAlign_Left, verticalAlignment=VAlign_Center, padding=<ltrb 5 5 5 5>}}
  }
  Spacer {
   widgetClassPath=&Image
   bIsVariable=true
   parent=HorizontalBox_78
   widget=Spacer
   properties {brush {
     bIsDynamicallyLoaded=false
     drawAs=Image
     imageType=NoImage
     mirroring=NoMirror
     resourceName=None
     resourceObject=null
     tiling=NoTile
     imageSize=<1 32>
     margin=<ltrb 0 0 0 0>
     outlineSettings {
      bUseBrushTransparency=false
      roundingType=HalfHeightRadius
      width=0
      color {colorUseRule=UseColor_Specified, specifiedColor=<rgba 0 0 0 0>}
      cornerRadii=<0 0 0 0>
     }
     tintColor {colorUseRule=UseColor_Specified, specifiedColor=<rgba 0.0103 0.0103 0.0103 1>}
     uVRegion {bIsValid=false, max=<0 0>, min=<0 0>}
    }}
   slot {class=HorizontalBoxSlot, properties {padding=<ltrb 5 0 5 0>}}
  }
  FortCommonQuestTagQueryMenu {
   widgetClassPath=&/FortCommonQuestTagQueryMenu
   bIsVariable=true
   parent=HorizontalBox_78
   widget=FortCommonQuestTagQueryMenu
   properties {toolTipText=Library of commonly used tag queries}
   slot {class=HorizontalBoxSlot, properties {horizontalAlignment=HAlign_Center, verticalAlignment=VAlign_Center, padding=<ltrb 5 5 5 5>}}
  }
  Button_CommonTagQueriesHelp {
   widgetClassPath=&QuestEditorButton
   bIsVariable=true
   parent=HorizontalBox_78
   widget=Button_CommonTagQueriesHelp
   properties {
    bShowIcon=true
    slateIconName=GraphEditor.Documentation_16x
    style=/Game/NeverCook/EditorUtilities/QuestTools/QuestEditor/QuestEditorButtons/Styles/QuestEditorButtonSlateIconStyle.QuestEditorButtonSlateIconStyle_C
    toolTipText=Guide to adding tag queries to the quest editor
   }
   slot {class=HorizontalBoxSlot, properties {verticalAlignment=VAlign_Center, padding=<ltrb 5 5 5 5>}}
  }
  Spacer_1 {
   widgetClassPath=&Image
   bIsVariable=true
   parent=VerticalBox_0
   widget=Spacer_1
   properties {brush {
     bIsDynamicallyLoaded=false
     drawAs=Image
     imageType=NoImage
     mirroring=NoMirror
     resourceName=None
     resourceObject=null
     tiling=NoTile
     imageSize=<1 1>
     margin=<ltrb 0 0 0 0>
     outlineSettings {
      bUseBrushTransparency=false
      roundingType=HalfHeightRadius
      width=0
      color {colorUseRule=UseColor_Specified, specifiedColor=<rgba 0 0 0 0>}
      cornerRadii=<0 0 0 0>
     }
     tintColor {colorUseRule=UseColor_Specified, specifiedColor=<rgba 0 0 0 1>}
     uVRegion {bIsValid=false, max=<0 0>, min=<0 0>}
    }}
   slot {class=VerticalBoxSlot}
  }
 ]
}
